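; CosmeticTestInterface is a cosmetics Blueprint interface, named as a test, that lets callers pass a float value to a cosmetic.
; It declares a single function, setCosmeticFloat0, which takes one double-precision float and returns nothing.
; The file holds only the signature, so what the value controls is left to each implementer.
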(fn setCosmeticFloat0
  (params (float :type Float%20%28double-precision%29)))
